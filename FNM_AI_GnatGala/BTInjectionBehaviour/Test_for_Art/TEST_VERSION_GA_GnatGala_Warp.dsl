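; Test version of the GnatGala Warp gameplay ability, filed under the AI's behaviour tree injection test assets for art.
; Activation only calls the parent implementation, so this asset defines no warp logic of its own.
; CanActivateAbility is overridden to always return true, without calling the parent's checks, and the end-ability event has no logic.
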(event Ability|EventActivateAbility
  (|CallParentFunction))

(event Ability|EventOnEndAbility
  (params (bWasCancelled :type Boolean)))

(fn CanActivateAbility
  (params (ActorInfo :type Gameplay%20Ability%20Actor%20Info%20Structure) (Handle :type Gameplay%20Ability%20Spec%20Handle%20Structure))
  (return 0 true))
